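# Physics asset for the dynamic beanie on the female medium snowboard hat cosmetic, giving the hat secondary motion on the head.
# It has two capsule bodies, one on the dyn_hat bone and a small one on the head bone.
# A single constraint joins dyn_hat to head with limited swing and twist, so the hat can wobble within a narrow range.
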PhysicsAsset /BRCosmetics/Accessories/Hats/F_MED_Snow_Board/Meshes/F_MED_Snow_Board_Dyn_Beanie
bodies {
 dyn_hat [{shapeType=Capsule, length=7.6472, radius=5.7577, center=<2.6149 -2.268 0.0001>, rotation=<rot 90 0 0>}]
 head [{shapeType=Capsule, length=0.7176, radius=2.0156, center=<5.9753 -1.989 -0.1792>, rotation=<rot 0 0 -90>}]
}
constraints [{bone1Name=dyn_hat, bone2Name=head, swing1LimitDegrees=30, swing2LimitDegrees=30, twistLimitDegrees=15}]
