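# Physics asset for the F_MED_Snow_Board hat cosmetic, defining collision and joint limits for the hair that moves with it.
# Capsule bodies cover the neck, head, a two-link front hair chain on a base bone, and a four-link side hair chain.
# Both hair chains hang from the head, with swing limits loosening toward the tips and twist kept tight.
# Head-to-neck and front-hair-base-to-head joints leave swing and twist free.
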PhysicsAsset /BRCosmetics/Accessories/Hats/F_MED_Snow_Board/Meshes/F_MED_Snow_Board_Hat_Physics
bodies {
 base_FrontHair [{shapeType=Capsule, length=0.505, radius=0.505, shapeName=base_FrontHair_capsule, rotation=<rot 0 -90 0>}]
 dyn_FrontHair_01 [{shapeType=Capsule, length=1, radius=1, shapeName=dyn_FrontHair_01_capsule, center=<1 0 0>, rotation=<rot 90 0 0>}]
 dyn_FrontHair_02 [{shapeType=Capsule, length=1, radius=1, shapeName=dyn_FrontHair_02_capsule, center=<2 0 0>, rotation=<rot 90 0 0>}]
 dyn_hair_side_01 [{shapeType=Capsule, length=1, radius=1, center=<1 0 0>, rotation=<rot 90 0 0>}]
 dyn_hair_side_02 [
  {shapeType=Capsule, length=4.4116, radius=1.4169, center=<0.207 0 -0.0001>}
  {shapeType=Capsule, length=4.4116, radius=1.4169, center=<2.7413 0 0.0001>}
 ]
 dyn_hair_side_03 [{shapeType=Capsule, length=4.4116, radius=1.4169, center=<1.7265 0 -0.0001>}]
 dyn_hair_side_04 [{shapeType=Capsule, length=4.4116, radius=1.4169, center=<1.7265 0 -0.0001>}]
 head [{shapeType=Capsule, length=0.7176, radius=2.0156, center=<5.9753 -1.989 -0.1792>, rotation=<rot 0 0 -90>}]
 neck_01 [{shapeType=Capsule, length=0, radius=1.1017, rotation=<rot 0 -90 0>}]
}
constraints [
 {bone1Name=dyn_hair_side_01, bone2Name=head, swing1LimitDegrees=30, swing2LimitDegrees=30, twistLimitDegrees=2}
 {bone1Name=dyn_hair_side_02, bone2Name=dyn_hair_side_01, swing1LimitDegrees=60, swing2LimitDegrees=60, twistLimitDegrees=2}
 {bone1Name=dyn_hair_side_03, bone2Name=dyn_hair_side_02, swing1LimitDegrees=90, swing2LimitDegrees=90, twistLimitDegrees=2}
 {bone1Name=dyn_hair_side_04, bone2Name=dyn_hair_side_03, swing1LimitDegrees=90, swing2LimitDegrees=90, twistLimitDegrees=2}
 {
  bone1Name=head
  bone2Name=neck_01
  swing1LimitDegrees=45
  swing1Motion=Free
  swing2LimitDegrees=45
  swing2Motion=Free
  twistLimitDegrees=45
  twistMotion=Free
 }
 {
  bone1Name=base_FrontHair
  bone2Name=head
  swing1LimitDegrees=45
  swing1Motion=Free
  swing2LimitDegrees=45
  swing2Motion=Free
  twistLimitDegrees=45
  twistMotion=Free
 }
 {bone1Name=dyn_FrontHair_01, bone2Name=base_FrontHair, swing1LimitDegrees=2, swing2LimitDegrees=2, twistLimitDegrees=1}
 {bone1Name=dyn_FrontHair_02, bone2Name=dyn_FrontHair_01, swing1LimitDegrees=10, swing2LimitDegrees=8, twistLimitDegrees=3}
]
