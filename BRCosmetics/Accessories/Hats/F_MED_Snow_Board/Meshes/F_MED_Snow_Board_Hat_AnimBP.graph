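# Post-process style AnimBlueprint for the F_MED_Snow_Board hat, copying the attached parent mesh's pose and layering physics for collar, hair and beanie.
# A spine-relative collar RigidBody pose is cached, then feeds head-relative hat and zero-alpha beanie RigidBody nodes, with additive hair-bone offsets.
# The update event inverts bRigidBodyEnabled to toggle the base_hair_side offset as a physics-off correction.
# Targeting events set isTargetingAlpha, which the AnimGraph never reads.
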AnimBlueprint /BRCosmetics/Accessories/Hats/F_MED_Snow_Board/Meshes/F_MED_Snow_Board_Hat_AnimBP
target_skeleton=&/F_MED_Snow_Board_Hat_Skeleton
graphs [
 AnimGraph {nodes [
   n1 = "Transform (Modify) Bone - Bone: base_hair_side"!(
    ComponentPose=Local To Component(LocalPose=Use cached pose 'prehairCache'())
    bAlphaBoolEnabled:bool=Get Rigid Body Disable() -> (Rigid Body Disable:bool)
    Translation="-0.0442,-0.5,0"
    Rotation="0,-15,0"
    Scale="1,1,1"
   ) -> (Pose) {
    alphaInputType=Bool
    rotationMode=BMM_Additive
    rotationSpace=BCS_BoneSpace
    translationMode=BMM_Additive
    translationSpace=BCS_BoneSpace
    alphaBoolBlend {blendOption=Linear}
    boneToModify {boneName=base_hair_side}
    rotation {yaw=-15}
    translation=<-0.0442 -0.5>
   }
   n2 = RigidBody(ComponentPose=@n1) {
    bOverrideWorldGravity=true
    overridePhysicsAsset=&/F_MED_Snow_Board_Hat_Physics
    simulationSpace=BaseBoneSpace
    alphaBoolBlend {blendOption=Linear}
    baseBoneRef {boneName=head}
    componentAppliedLinearAccClamp=<2000 2000 2000>
    componentLinearAccScale=<0.1 0.05 0.05>
    componentLinearVelScale=<0.1 0.05 0.05>
    overrideWorldGravity {z=-980}
   }
   n3 = RigidBody(ComponentPose=@n2) {
    alpha=0
    bOverrideWorldGravity=true
    overridePhysicsAsset=&/F_MED_Snow_Board_Dyn_Beanie
    simulationSpace=BaseBoneSpace
    alphaBoolBlend {blendOption=Linear}
    baseBoneRef {boneName=head}
    componentAppliedLinearAccClamp=<2000 2000 2000>
    componentLinearAccScale=<2 1 1>
    componentLinearVelScale=<2 1 1>
    overrideWorldGravity {z=-980}
   }
   n4 = "Transform (Modify) Bone - Bone: dyn_hair_side_01"(ComponentPose=@n3, Alpha=1, Translation="-1,-1,0", Rotation="0,0,0", Scale="1,1,1") {
    alpha=0
    translationMode=BMM_Additive
    translationSpace=BCS_BoneSpace
    alphaBoolBlend {blendOption=Linear}
    boneToModify {boneName=dyn_hair_side_01}
    translation=<-1 -1>
   }
   Output Pose(Result=Component To Local(ComponentPose=@n4)) {layerGroup=None}
   n5 = RigidBody(ComponentPose=Local To Component(LocalPose=Copy Pose From Mesh() {bUseAttachedParent=true})) {
    bOverrideWorldGravity=true
    overridePhysicsAsset=/BRCosmetics/Characters/Player/Female/Medium/Bodies/F_MED_Snow_Board/Meshes/F_MED_Snow_Board_Dyn_Collar
    simulationSpace=BaseBoneSpace
    alphaBoolBlend {blendOption=Linear}
    baseBoneRef {boneName=spine_05}
    componentAppliedLinearAccClamp=<1000 1000 1000>
    componentLinearAccScale=<2 1 1>
    componentLinearVelScale=<2 1 1>
    overrideWorldGravity {z=-980}
   }
   Save cached pose 'prehairCache'(Pose=Component To Local(ComponentPose=@n5)) {cacheName=prehairCache}
  ]}
 EventGraph {nodes [
   Event BlueprintUpdateAnimation()
   > Set Rigid Body Disable(Rigid Body Disable:bool=NOT Boolean(A=Get bRigidBodyEnabled())) -> (Output_Get:bool)
   Event OnBeginTargeting() {is=title K2Node_Event}
   > Set isTargetingAlpha(isTargetingAlpha:double=0.5) -> (Output_Get:double)
   Event OnEndTargeting() {is=title K2Node_Event}
   > Set isTargetingAlpha(isTargetingAlpha=1)
   TryGetPawnOwner()
  ]}
]
variables [
 isTargetingAlpha {type=number, default=1}
 Rigid Body Disable {type=boolean, default=true}
]
